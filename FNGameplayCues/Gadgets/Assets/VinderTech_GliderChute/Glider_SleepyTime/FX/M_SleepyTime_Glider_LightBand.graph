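# Translucent, two-sided mesh-particle material for the light band effect on the SleepyTime glider.
# Curve atlas rows and top/bottom thickness parameters reshape UVs into band coordinates that drive warm color layers, scaled by particle color into emissive.
# Alpha combines a fire mask with a sine-shaped glow, faded by a curve along V and particle alpha.
# SplineThicken sets band width from a scalar times the Thickness dynamic parameter.
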Material /FNGameplayCues/Gadgets/Assets/VinderTech_GliderChute/Glider_SleepyTime/FX/M_SleepyTime_Glider_LightBand
graph {
 properties {bUsedWithMeshParticles=true, blendMode=BLEND_Translucent, pixelDepthOffsetMode=PDOM_Legacy, twoSided=true, usedShadingModels=Default Lit}
 nodes [
  n1 = ParticleColor()
  n2 = CurveAtlasRowParameter(CurveTime=ComponentMask(None=TextureCoordinate()) {g=true}) {
   atlas=&AthenaCurveAtlas01
   curve=/Game/Accessories/FORT_Backpacks/M_MED_Sleepy_Time_Backpack/FX/Curve_SleepyTime
   defaultValue=127
   parameterName=Param_1
  }
  n3 = Multiply(A=@n2.R)
  n4 = ScalarParameter() {parameterName=LightTopThickness}
  n5 = Constant() {r=2}
  n6 = TextureCoordinate()
  n7 = LinearInterpolate(A=@n5, B=OneMinus(None=@n5), Alpha=ComponentMask(None=@n6) {r=true})
  n8 = ScalarParameter() {defaultValue=-1, parameterName=LightBottomThickness}
  n9 = CurveAtlasRowParameter(CurveTime=ComponentMask(None=@n6) {g=true}) {
   atlas=&AthenaCurveAtlas01
   curve=/Game/Gadgets/Assets/VinderTech_GliderChute/Glider_SleepyTime/FX/Curve_SleepyTime_Glider
   defaultValue=128
   parameterName=Param_3
  }
  n10 = LinearInterpolate(A=LinearInterpolate(A=@n4, B=OneMinus(None=@n4), Alpha=@n7), B=LinearInterpolate(A=@n8, B=OneMinus(None=@n8), Alpha=@n7), Alpha=@n9.A)
  n11 = CurveAtlasRowParameter(CurveTime=@n10) {
   atlas=&AthenaCurveAtlas01
   curve=&Curve_SleepyTime_Glider
   defaultValue=128
   parameterName=Param_2
  }
  n12 = OneMinus(None=@n11.B)
  n13 = Add(
   A=Multiply(A=Constant3Vector() {constant=<rgba 1 0.8273 0 0>}, B=Multiply(A=@n11.G, B=@n12))
   B=Multiply(A=@n11.B, B=Constant3Vector() {constant=<rgba 0.338 0.1408 0 0>})
  ) {constB=0}
  n14 = Add(A=Multiply(A=LinearInterpolate(Alpha=@n3) {constB=3}, B=@n13), B=@n13)
  n15 = CurveAtlasRowParameter(CurveTime=@n10) {
   atlas=&AthenaCurveAtlas01
   curve=&Curve_SleepyTime_Glider
   defaultValue=128
   parameterName=Param_3
  }
  n16 = Add(A=@n15.G, B=@n15.G)
  n17 = Multiply(A=Clamp(None=OneMinus(None=@n16)), B=Constant3Vector() {constant=<rgba 1 0.644 0 0>})
  MaterialFunctionCall(Emissive=Multiply(A=Add(A=@n14, B=@n17), B=@n1.RGB)) {materialFunction=&MF_FX_EmissiveAdjustments}
  n18 = Multiply(A=ScalarParameter() {defaultValue=50, parameterName=Tickness}, B=DynamicParameter() {paramNames [Thickness,Param2,Param3,Param4]}.Thickness)
  MaterialFunctionCall(WidthBase=@n18, WidthTip=@n18) {materialFunction=&SplineThicken}
  n19 = Multiply(A=@n10)
  n20 = LinearInterpolate(
   A=ScalarParameter() {defaultValue=0.6, parameterName=Param_1}
   B=ScalarParameter() {defaultValue=1.8, parameterName=Param_2}
   Alpha=ComponentMask(None=TextureCoordinate()) {g=true}
  ) {constA=0.6, constB=3}
  n21 = Multiply(A=@n16, B=TextureSample(UVs=AppendVector(A=@n19, B=@n20)) {texture=/Game/Effects/Fort_Effects/Textures/Fire/T_Fire_Mask}.R)
  n22 = Constant() {r=-0.2}
  n23 = TextureCoordinate()
  n24 = Sine(None=LinearInterpolate(A=@n22, B=Subtract(B=@n22) {constA=0.5, constB=0.5}, Alpha=ComponentMask(None=@n23) {r=true}) {constA=-0.5})
  n25 = Multiply(A=OneMinus(None=@n16), B=Multiply(A=Power(Base=@n24), B=ScalarParameter() {defaultValue=0.06, parameterName=GlowIntensity}) {constB=0.2})
  n26 = CurveAtlasRowParameter(CurveTime=ComponentMask(None=TextureCoordinate()) {g=true}) {
   atlas=&AthenaCurveAtlas01
   curve=&Curve_SleepyTime_Glider
   defaultValue=128
   parameterName=Param_3
  }
  Multiply(A=@n1.A, B=Multiply(A=Add(A=@n21, B=@n25), B=@n26.R))
 ]
}
